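# Material function in the Mr. Mite gadget prop materials, exposed to the library, that calculates a hexagonal distance gradient from UVs.
# The In_UV input has a centered texture-coordinate preview, and named reroutes carry the UV, its absolute value and a (1, √3) vector.
# Folding by absolute value gives symmetry, and the maximum of a half-scaled dot product and a red-channel component outputs HexagonShapeGradient.
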MaterialFunction /BRCosmetics/Animation/Props/Gadgets_MrMite/Materials/MF_CalcHexDistance
graph {
 properties {bExposeToLibrary=true, description=Calculate the hexagonal shape, userExposedCaption=Calculate the hexagonal shape}
 nodes [
  n1 = NamedRerouteUsage()
  n2 = Multiply(A=NamedRerouteUsage()) {constB=0.5}
  n3 = NamedRerouteUsage()
  FunctionOutput(None=Max(A=DotProduct(A=@n1, B=@n2), B=ComponentMask(None=@n3) {r=true})) {outputName=HexagonShapeGradient}
  n4 = TextureCoordinate()
  n5 = Constant2Vector() {g=0.5, r=0.5}
  NamedRerouteDeclaration(None=FunctionInput(Preview=Subtract(A=@n4, B=@n5)) {inputName=In_UV, inputType=FunctionInput_Vector2}) {name=In_UV, nodeColor=<rgba 0.5469 0 1 1>}
  NamedRerouteDeclaration(None=Abs(None=NamedRerouteUsage())) {name=Abs_UV, nodeColor=<rgba 0 1 0.2266 1>}
  NamedRerouteDeclaration(None=AppendVector(A=Constant() {r=1}, B=Constant() {r=1.7321})) {name=s, nodeColor=<rgba 1 0 0.0937 1>}
 ]
}
