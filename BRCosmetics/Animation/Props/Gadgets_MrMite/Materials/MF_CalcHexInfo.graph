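# Material function in the MrMite gadget prop folder that maps input UVs onto a hexagonal grid, returning per-cell hex info as a four-component vector.
# It rounds UVs against two staggered rectangular grids with a 1 by √3 ratio to get candidate hex centers, then subtracts those to get offsets.
# A squared-distance comparison via dot products selects the nearer candidate's packed data, avoiding square roots.
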MaterialFunction /BRCosmetics/Animation/Props/Gadgets_MrMite/Materials/MF_CalcHexInfo
graph {nodes [
  n1 = ComponentMask(None=NamedRerouteUsage()) {g=true, r=true}
  n2 = ComponentMask(None=NamedRerouteUsage()) {a=true, b=true}
  n3 = AppendVector(A=ComponentMask(None=NamedRerouteUsage()) {a=true, b=true}, B=ComponentMask(None=NamedRerouteUsage()) {a=true, b=true})
  n4 = ComponentMask(None=NamedRerouteUsage()) {g=true, r=true}
  n5 = ComponentMask(None=NamedRerouteUsage()) {g=true, r=true}
  FunctionOutput(None=If(A=DotProduct(A=@n1, B=@n1), B=DotProduct(A=@n2, B=@n2), "A > B"=@n3, "A == B"=@n3, "A < B"=AppendVector(A=@n4, B=@n5)))
  NamedRerouteDeclaration(None=Constant2Vector() {g=1.7321, r=1}) {name=sqrt3, nodeColor=<rgba 0 0.0391 1 1>}
  n6 = NamedRerouteDeclaration(None=FunctionInput() {inputName=In_UVs, inputType=FunctionInput_Vector2}) {name=UV, nodeColor=<rgba 1 0.2812 0 1>}
  n7 = AppendVector(A=@n6, B=Subtract(A=@n6, B=Constant2Vector() {g=1, r=0.5}))
  n8 = AppendVector(A=NamedRerouteUsage(), B=NamedRerouteUsage())
  NamedRerouteDeclaration(None=Round(None=Divide(A=@n7, B=@n8))) {name=hexcenter, nodeColor=<rgba 0.3594 1 0 1>}
  n9 = NamedRerouteUsage()
  n10 = Multiply(A=ComponentMask(None=NamedRerouteUsage()) {g=true, r=true}, B=NamedRerouteUsage())
  n11 = NamedRerouteUsage()
  n12 = NamedRerouteUsage()
  n13 = Multiply(A=Add(A=ComponentMask(None=@n12) {a=true, b=true}, B=Constant2Vector() {g=0.5, r=0.5}), B=NamedRerouteUsage())
  NamedRerouteDeclaration(None=AppendVector(A=Subtract(A=@n9, B=@n10), B=Subtract(A=@n11, B=@n13))) {name=offset, nodeColor=<rgba 1 0 0.6797 1>}
 ]}
