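# MF_HexHash is a material function in the Gadgets_MrMite prop folder that turns a 2D UV input into a pseudo-random hash value, output as HashNoise.
# It uses the classic sine hash: dot the UV with a fixed constant vector, take the sine, multiply by a large constant, and keep the fractional part.
# The In_UV input falls back to texture coordinates when left unconnected.
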MaterialFunction /BRCosmetics/Animation/Props/Gadgets_MrMite/Materials/MF_HexHash
graph {nodes [
  n1 = TextureCoordinate()
  n2 = Sine(None=DotProduct(
    A=FunctionInput(Preview=@n1) {bUsePreviewValueAsDefault=true, inputName=In_UV, inputType=FunctionInput_Vector2}
    B=Constant2Vector() {g=78.233, r=12.9898}
   ))
  n3 = Constant() {r=43758.5469}
  FunctionOutput(None=Frac(None=Multiply(A=@n2, B=@n3))) {outputName=HashNoise}
 ]}
