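# MF_ModForHex is a material function that provides an OpenGL-style mod, named for hex math and stored with the MrMite gadget prop materials.
# It takes scalar inputs X and Y, routed through named reroutes, and builds the floored modulo X - Y * floor(X / Y) as its single output.
# Flooring the quotient, rather than truncating it, is what distinguishes this from an fmod-style remainder for negative values.
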MaterialFunction /BRCosmetics/Animation/Props/Gadgets_MrMite/Materials/MF_ModForHex
graph {
 properties {description=OpenGL mod function, userExposedCaption=OpenGL mod function}
 nodes [
  n1 = Floor(None=Divide(A=NamedRerouteUsage(), B=NamedRerouteUsage()))
  n2 = NamedRerouteUsage()
  FunctionOutput(None=Subtract(A=NamedRerouteUsage(), B=Multiply(A=@n1, B=@n2))) {outputName=ModForHex}
  NamedRerouteDeclaration(None=FunctionInput() {inputName=X, inputType=FunctionInput_Scalar}) {name=X, nodeColor=<rgba 0.6875 1 0 1>}
  NamedRerouteDeclaration(None=FunctionInput() {inputName=Y, inputType=FunctionInput_Scalar, sortPriority=1}) {name=Y, nodeColor=<rgba 0.9922 0 1 1>}
 ]
}
