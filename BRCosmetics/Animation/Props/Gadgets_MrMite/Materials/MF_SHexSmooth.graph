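# Material function in the MrMite gadget prop materials that outputs an antialiased line mask from a scalar distance field, likely for hexagon shapes.
# It takes In_Distance and In_Offset through named reroutes and uses the absolute value of their difference.
# An inverted SmoothStep peaks where that difference is zero and fades out over a width inversely proportional to vertical screen resolution, keeping the edge a near-constant pixel softness.
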MaterialFunction /BRCosmetics/Animation/Props/Gadgets_MrMite/Materials/MF_SHexSmooth
graph {nodes [
  n1 = Constant() {r=9}
  n2 = ComponentMask(None=MaterialFunctionCall() {materialFunction=&ScreenResolution}.Buffer Resolution) {g=true}
  n3 = Subtract(A=NamedRerouteUsage(), B=NamedRerouteUsage())
  FunctionOutput(None=SmoothStep(Min=Divide(A=@n1, B=@n2), Value=Abs(None=@n3)) {constMax=0}) {outputName=AntiAliased}
  NamedRerouteDeclaration(None=FunctionInput() {inputName=In_Distance, inputType=FunctionInput_Scalar, sortPriority=1}) {name=Distance, nodeColor=<rgba 0.5938 0 1 1>}
  NamedRerouteDeclaration(None=FunctionInput() {inputName=In_Offset, inputType=FunctionInput_Scalar}) {name=Offset, nodeColor=<rgba 0 1 0.2734 1>}
 ]}
