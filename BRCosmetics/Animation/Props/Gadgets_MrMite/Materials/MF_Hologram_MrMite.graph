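# MrMite gadget hologram material function that adds an animated hexagon tech effect and see-through hologram shading to incoming material attributes.
# Panned UVs feed hex distance and hash helpers that build time-animated isolines and borders, with a tech-art mask controlling regions.
# Hologram shading combines skylight cubemap, sunlight passthrough, Fresnel, time-of-day scaling and scene depth, with low-quality fallbacks.
# Output overrides emissive, base color (pre-skinned normal picks photo backing) and Bayer-dithered opacity mask.
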MaterialFunction /BRCosmetics/Animation/Props/Gadgets_MrMite/Materials/MF_Hologram_MrMite
graph {nodes [
  BlendMaterialAttributes(A=NamedRerouteUsage(), B=NamedRerouteUsage(), Alpha=ComponentMask(None=NamedRerouteUsage()) {r=true})
  n1 = Subtract(B=NamedRerouteUsage()) {constA=0.51}
  n2 = NamedRerouteUsage()
  n3 = Subtract(B=NamedRerouteUsage()) {constA=0.5}
  n4 = Power(Base=OneMinus(None=Max(B=@n3))) {constExponent=20}
  n5 = Add(
   A=NamedRerouteUsage()
   B=Multiply(A=VectorParameter() {group=MrMiteTechArt, parameterName=OffColor, defaultValue=<rgba 0 0.1 0.09 1>}.RGB, B=NamedRerouteUsage())
  )
  n6 = NamedRerouteUsage()
  n7 = NamedRerouteUsage()
  n8 = Add(B=NamedRerouteUsage()) {constA=0.2}
  n9 = SmoothStep(Min=@n8, Max=Subtract(A=@n8, B=NamedRerouteUsage()), Value=NamedRerouteUsage())
  n10 = Multiply(
   A=Add(A=SmoothStep(Max=@n1, Value=@n2) {constMin=0.51}, B=Multiply(A=@n4) {constB=1.5})
   B=Add(A=Add(A=@n5, B=@n6), B=Multiply(A=@n7, B=@n9))
  )
  n11 = Multiply(A=ScalarParameter() {defaultValue=0.5, group=MrMiteTechArt, parameterName=EmissiveStrength}, B=@n10)
  n12 = Add(A=ComponentMask(None=NamedRerouteUsage()) {r=true}, B=NamedRerouteUsage())
  n13 = NamedRerouteUsage()
  n14 = LinearInterpolate(A=NamedRerouteUsage(), Alpha=Add(A=ComponentMask(None=@n13) {b=true}, B=ComponentMask(None=@n13) {g=true}))
  n15 = Constant() {r=3}
  n16 = LinearInterpolate(A=@n10, B=NamedRerouteUsage(), Alpha=ComponentMask(None=NamedRerouteUsage()) {b=true})
  n17 = VertexInterpolator(VS=PreSkinnedNormal())
  n18 = LinearInterpolate(
   A=VectorParameter() {group=MrMiteTechArt, parameterName=PhotoBackColour, defaultValue=<rgba 0.6979 0.6979 0.6979 1>}.RGB
   B=NamedRerouteUsage()
   Alpha=Step(X=ComponentMask(None=@n17.PS) {g=true})
  )
  n19 = ComponentMask(None=NamedRerouteUsage()) {g=true}
  FunctionOutput(None=SetMaterialAttributes!(
    MaterialAttributes=NamedRerouteUsage()
    EmissiveColor=LinearInterpolate(B=@n11, Alpha=@n12)
    OpacityMask=MaterialFunctionCall!(In=@n14, Iterations=@n15, Scale) {materialFunction=/Game/Packages/Fortress_SharedMaterials/Base_Material_Functions/MF_DitherBayerNoTex}.Result
    BaseColor=LinearInterpolate(A=@n16, B=@n18, Alpha=@n19)
   ) {attributeSetTypes [#1, #2, #3]})
  LinearInterpolate(A=NamedRerouteUsage(), Alpha=ComponentMask(None=NamedRerouteUsage()) {r=true})
  LinearInterpolate(A=NamedRerouteUsage(), Alpha=ComponentMask(None=NamedRerouteUsage()) {r=true})
  MaterialFunctionCall(Alpha Threshold=@n14) {materialFunction=&DitherTemporalAA_Fortnite}
  MaterialFunctionCall!(ObjectPivotLocation, ObjectPositionOffset) {materialFunction=/Game/Characters/Player/Common/MaterialFunctions/Superlevel/ImprovedStabilizedScreenSpaceUVs}
  Multiply()
  n20 = Constant() {r=1}
  n21 = MaterialFunctionCall(Value_Morning=@n20, Value_Day=@n20, Value_Evening=@n20, Value_Night=ScalarParameter() {defaultValue=5, parameterName=Night Boost}) {materialFunction=&Scalar_Time_of_Day}
  Multiply(A=@n21.Result) {constB=0.5}
  n22 = Saturate(None=CollectionParameter() {collection=&FortniteMaterialParameters, parameterName=SunLightColor})
  n23 = Multiply(
   A=ScalarParameter() {defaultValue=0.02, parameterName=Sunlight Passthrough}
   B=Multiply(A=ComponentMask(None=@n22) {b=true, g=true, r=true}, B=NamedRerouteUsage())
  )
  n24 = Clamp(None=MaterialFunctionCall() {materialFunction=/Game/Packages/Fortress_Sky/Fortnite_SunScreenSpaceFalloff}.Dot*-1) {minDefault=0.5}
  NamedRerouteDeclaration(None=QualitySwitch(Default=Multiply(A=@n23, B=@n24), Low=Constant())) {name=SunlightPassthrough, nodeColor=<rgba 0.5 1 0 1>}
  NamedRerouteDeclaration(None=FunctionInput(Preview=TextureSample() {samplerType=SAMPLERTYPE_Masks, texture=/BRCosmetics/Animation/Props/Gadgets_MrMite/Textures/T_MrMite_Gadget_FX}.RGB) {inputName=In_TechArtMask, sortPriority=1}) {name=TechArtMask, nodeColor=<rgba 1 0.172 0.0441 1>}
  n25 = TextureCoordinate()
  n26 = NamedRerouteUsage()
  n27 = Time()
  n28 = NamedRerouteUsage()
  n29 = NamedRerouteDeclaration(None=ScalarParameter() {defaultValue=0.1, group=MrMiteTechArt, parameterName=XSpeed}) {name=XSpeed, nodeColor=<rgba 0.073 0.0656 1 1>}
  n30 = NamedRerouteDeclaration(None=ScalarParameter() {group=MrMiteTechArt, parameterName=YSpeed}) {name=YSpeed, nodeColor=<rgba 0.073 0.0656 1 1>}
  NamedRerouteDeclaration(None=Panner(Coordinate=Multiply(A=@n25, B=@n26), Time=Multiply(A=@n27, B=@n28), Speed=AppendVector(A=@n29, B=@n30))) {name=AnimatedUVs, nodeColor=<rgba 1 0.5156 0 1>}
  NamedRerouteDeclaration(None=MaterialFunctionCall!(In_UVs=NamedRerouteUsage()) {materialFunction=&/MF_CalcHexInfo}.Result) {name=HexInfo, nodeColor=<rgba 0.1953 0 1 1>}
  n31 = ComponentMask(None=NamedRerouteUsage()) {g=true, r=true}
  NamedRerouteDeclaration(None=Add(
    A=MaterialFunctionCall!(In_UV=@n31) {materialFunction=&/MF_CalcHexDistance}.HexagonShapeGradient
    B=NamedRerouteUsage()
   )) {name=TotalDist, nodeColor=<rgba 0.125 1 0 1>}
  n32 = NamedRerouteUsage()
  NamedRerouteDeclaration(None=MaterialFunctionCall!(In_UV=ComponentMask(None=@n32) {a=true, b=true}) {materialFunction=&/MF_HexHash}.HashNoise) {name=RandomValue, nodeColor=<rgba 1 0 0.4453 1>}
  NamedRerouteDeclaration(None=ScalarParameter() {defaultValue=0.1, group=MrMiteTechArt, parameterName=TimeMultiplier}) {name=TimeMultiplier, nodeColor=<rgba 0.1151 1 0.8947 1>}
  n33 = MaterialFunctionCall(Float4=NamedRerouteUsage()) {materialFunction=&BreakOutFloat4Components}
  n34 = Arctangent2!(Y=@n33.G, X=@n33.G)
  n35 = Multiply(A=NamedRerouteUsage()) {constB=0.5}
  NamedRerouteDeclaration(None=Add(A=Add(A=@n34, B=@n35), B=NamedRerouteUsage())) {name=Angle, nodeColor=<rgba 1 0.9141 0 1>}
  n36 = Multiply(
   A=MaterialFunctionCall!(In_Offset=NamedRerouteUsage(), In_Distance=NamedRerouteUsage()) {materialFunction=&/MF_SHexSmooth}.AntiAliased
   B=NamedRerouteUsage()
  )
  n37 = NamedRerouteUsage()
  n38 = Constant() {r=2}
  n39 = Step(
   Y=Add(A=Multiply(A=@n37) {constB=0.5}) {constB=3}
   X=MaterialFunctionCall!(
    X=NamedRerouteUsage()
    Y=MaterialFunctionCall(Multiplier=@n38) {materialFunction=&Pi}.Result
   ) {materialFunction=&/MF_ModForHex}.ModForHex
  )
  n40 = Multiply(
   A=MaterialFunctionCall(In_Offset=NamedRerouteUsage(), In_Distance=NamedRerouteUsage()) {materialFunction=&MF_SHexSmooth}.AntiAliased
   B=NamedRerouteUsage()
  )
  n41 = NamedRerouteUsage()
  n42 = NamedRerouteUsage()
  n43 = Multiply(A=NamedRerouteUsage()) {constB=2}
  n44 = Constant() {r=2}
  n45 = Step(
   Y=Add(A=Multiply(A=@n41) {constB=1.5}) {constB=4}
   X=MaterialFunctionCall(
    X=Add(A=@n42, B=@n43)
    Y=MaterialFunctionCall(Multiplier=@n44) {materialFunction=&Pi}.Result
   ) {materialFunction=&MF_ModForHex}.ModForHex
  )
  NamedRerouteDeclaration(None=Add(A=Multiply(A=@n36, B=@n39), B=Multiply(A=@n40, B=@n45))) {name=Isoline, nodeColor=<rgba 0.0312 0 1 1>}
  NamedRerouteDeclaration(None=Multiply(A=Time(), B=NamedRerouteUsage())) {name=TimeMult, nodeColor=<rgba 0.2891 1 0 1>}
  n46 = ScalarParameter() {defaultValue=20, group=AuraCop, parameterName=ScanlineAmountrRE}
  NamedRerouteDeclaration(None=MaterialFunctionCall!(
    Default=MaterialFunctionCall!(Scanline Amount=@n46) {materialFunction=/BRCosmetics/Animation/Props/Gadgets_TalonPane/Materials/MFScanlines_TalonPane}.Result
    Low/Mobile=Constant() {r=1}
   ) {materialFunction=/Game/Effects/Fort_Effects/Materials/_Functions/MF_LowMobile_PathingSwitch}.Result) {name=Scanlines, nodeColor=<rgba 1 0 0.1641 1>}
  n47 = NamedRerouteUsage()
  n48 = Multiply(A=NamedRerouteUsage()) {constB=8}
  NamedRerouteDeclaration(None=Sine(None=Add(A=@n47, B=@n48))) {name=SineOffset, nodeColor=<rgba 1 0 0.6094 1>}
  n49 = MaterialFunctionCall() {materialFunction=&ScreenResolution}
  NamedRerouteDeclaration(None=Divide(B=ComponentMask(None=@n49.Buffer Resolution) {g=true}) {constA=5}) {name=AA, nodeColor=<rgba 0 0.9297 1 1>}
  NamedRerouteDeclaration(None=ScalarParameter() {defaultValue=64, group=MrMiteTechArt, parameterName=XScale}) {name=XScale, nodeColor=<rgba 1 0.3901 0.2069 1>}
  NamedRerouteDeclaration(None=ScalarParameter() {defaultValue=64, group=MrMiteTechArt, parameterName=YScale}) {name=YScale, nodeColor=<rgba 1 0.2884 0.7293 1>}
  NamedRerouteDeclaration(None=AppendVector(A=NamedRerouteUsage(), B=NamedRerouteUsage())) {name=UVScale, nodeColor=<rgba 1 0.8203 0 1>}
  n50 = GetMaterialAttributes(None=NamedRerouteUsage()) {attributeGetTypes [#3, #1, #4]}
  NamedRerouteDeclaration(None=@n50.BaseColor) {name=In_BaseColor, nodeColor=<rgba 0.4766 0 1 1>}
  NamedRerouteDeclaration(None=@n50.EmissiveColor) {name=In_EmissiveColor, nodeColor=<rgba 0 1 0.1562 1>}
  NamedRerouteDeclaration(None=@n50.Opacity) {name=In_Opacity, nodeColor=<rgba 1 0.5391 0 1>}
  NamedRerouteDeclaration(None=ScalarParameter() {defaultValue=0.5, group=MrMiteTechArt, parameterName=OpacityScale}) {name=OpacityScale, nodeColor=<rgba 0.142 0.3788 1 1>}
  n51 = GetMaterialAttributes(None=NamedRerouteUsage()) {attributeGetTypes [#3, #1]}
  n52 = ScalarParameter() {defaultValue=3, parameterName=See Through Side Max}
  n53 = Saturate(None=Fresnel(ExponentIn=ScalarParameter() {defaultValue=3, parameterName=SeethroughSidesFres}) {baseReflectFraction=0})
  n54 = LinearInterpolate(A=Add(A=@n51.BaseColor, B=@n51.BaseColor), B=NamedRerouteUsage(), Alpha=Saturate(None=LinearInterpolate(B=@n52, Alpha=@n53) {constB=0}))
  n55 = NamedRerouteUsage()
  n56 = VectorParameter() {parameterName=ColorTint, defaultValue=<rgba 1 0.4673 0 2>}
  n57 = Multiply(A=Saturate(None=Multiply(A=@n54, B=@n55)), B=Multiply(A=@n56.RGB, B=@n56.RGB))
  n58 = NamedRerouteUsage()
  n59 = Constant() {r=0.7}
  n60 = Constant() {r=1}
  n61 = Constant() {r=0.7}
  n62 = Constant() {r=0.4}
  n63 = Add(
   A=@n51.EmissiveColor
   B=Multiply(
    A=Add(A=@n57, B=@n58)
    B=MaterialFunctionCall(Value_Morning=@n59, Value_Day=@n60, Value_Evening=@n61, Value_Night=@n62) {materialFunction=&Scalar_Time_of_Day}.Result
   )
  )
  n64 = Multiply(A=@n63, B=ScalarParameter() {defaultValue=8, parameterName=LumenBrightness})
  n65 = CameraPositionWS()
  n66 = WorldPosition()
  n67 = Add(A=Subtract(A=SceneDepth().R, B=Distance(A=@n65, B=@n66.XYZ)), B=Constant() {r=20})
  n68 = Multiply(
   A=LinearInterpolate(
    A=ScalarParameter() {defaultValue=0.2, group=Translucent, parameterName=MinOpacity}
    B=ScalarParameter() {defaultValue=0.8, group=Translucent, parameterName=MaxOpacity}
    Alpha=NamedRerouteUsage()
   ) {constA=0.2, constB=0.7}
   B=Saturate(None=Sign(None=@n67))
  )
  n69 = TextureSampleParameter2D() {group=Translucent, parameterName=TechArtRE, texture=/BRCosmetics/Animation/Props/Gadgets_TalonPane/Textures/T_Talonpane_Emote_Prop_TechArt}
  n70 = Power(Base=OneMinus(None=Fresnel() {baseReflectFraction=0, exponent=1})) {constExponent=3}
  n71 = Multiply(A=LinearInterpolate(A=LinearInterpolate(Alpha=@n69.R) {constA=0.05}, B=LinearInterpolate(Alpha=@n70) {constA=0.1}, Alpha=@n69.G), B=@n69.B)
  NamedRerouteDeclaration(None=SetMaterialAttributes!(
    MaterialAttributes=@n51.MaterialAttributes
    EmissiveColor=RayTracingQualitySwitch(Normal=@n63, RayTraced=@n64)
    Opacity=Multiply(A=@n68, B=@n71)
   ) {attributeSetTypes [#1, #4]}) {name=HologramMaterialAttributes, nodeColor=<rgba 1 0 0.9844 1>}
  NamedRerouteDeclaration(None=FunctionInput(Preview=TextureSample() {samplerType=SAMPLERTYPE_Masks, texture=&T_MrMite_Gadget_FX}.A) {bUsePreviewValueAsDefault=true, inputName=In_TechArtMask_A, inputType=FunctionInput_Scalar, sortPriority=2}) {name=TechArtMask_A, nodeColor=<rgba 1 0.172 0.0441 1>}
  NamedRerouteDeclaration(None=FunctionInput() {inputType=FunctionInput_MaterialAttributes}) {name=In_MaterialAttributes, nodeColor=<rgba 0 1 0.6953 1>}
  n72 = MaterialFunctionCall(Float3=OneMinus(None=CameraVectorWS())) {materialFunction=&BreakOutFloat3Components}
  n73 = MaterialFunctionCall(X=@n72.R, Y=@n72.R, Z=Add(A=@n72.B, B=ScalarParameter() {defaultValue=-1, parameterName=Offset SeeThrough})) {materialFunction=&MakeFloat3}
  n74 = ScalarParameter() {defaultValue=0.1, parameterName=HologramRoughness}
  n75 = Multiply(
   A=Clamp(None=SkyLightEnvMapSample(Direction=@n73.Result, Roughness=@n74)) {maxDefault=4, minDefault=0.5}
   B=ScalarParameter() {defaultValue=1, parameterName=CubemapBleed}
  )
  n76 = VectorParameter() {parameterName=ColorMorning, defaultValue=<rgba 0.2031 0.1723 0.1315 1>}
  n77 = VectorParameter() {parameterName=ColorNight, defaultValue=<rgba 0.0373 0.0718 0.1302 1>}
  NamedRerouteDeclaration(None=QualitySwitch(
    Default=Multiply(A=@n75, B=@n21.Result)
    Low=MaterialFunctionCall(Color_Morning=@n76.RGB, Color_Day=@n76.RGB, Color_Evening=@n76.RGB, Color_Night=@n77.RGB) {materialFunction=&Color_Time_of_Day}.Result
   )) {name=Cubemap, nodeColor=<rgba 0.8984 0 1 1>}
  NamedRerouteDeclaration(None=VectorParameter() {parameterName=BaseColor, defaultValue=<rgba 0.487 0.2245 0.0736 1>}.RGB) {name=BaseColor, nodeColor=<rgba 0 1 0.4375 1>}
  NamedRerouteDeclaration(None=ScalarParameter() {defaultValue=0.02, group=MrMiteTechArt, parameterName=BorderThickness}) {name=BorderThickness, nodeColor=<rgba 1 0.1172 0 1>}
  NamedRerouteDeclaration(None=ScalarParameter() {defaultValue=0.4, group=MrMiteTechArt, parameterName=IsolineOffset}) {name=IsolineOffset, nodeColor=<rgba 1 0.051 0.9467 1>}
  NamedRerouteDeclaration(None=ScalarParameter() {defaultValue=1, group=MrMiteTechArt, parameterName=IsolineOffset2}) {name=IsolineOffset2, nodeColor=<rgba 0.073 0.0656 1 1>}
 ]}
