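# Material function for the photos and string on the Mr. Mite gadget prop, adjusting their opacity within incoming material attributes.
# It lerps the incoming Opacity toward the OpacityOfPhotos parameter, using a mask's green channel scaled by OpacityLerpMultiply as alpha.
# Tech-art mask RGB and alpha inputs preview the gadget FX texture, with only the alpha input using that preview as its default.
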MaterialFunction /BRCosmetics/Animation/Props/Gadgets_MrMite/Materials/MF_PhotosAndString_MrMite
graph {nodes [
  n1 = GetMaterialAttributes(None=NamedRerouteUsage()) {attributeGetTypes [#1]}
  n2 = ScalarParameter() {defaultValue=1, group=MrMitePhotosCupAndString, parameterName=OpacityOfPhotos}
  n3 = Multiply(
   A=ScalarParameter() {defaultValue=1, group=MrMitePhotosCupAndString, parameterName=OpacityLerpMultiply}
   B=ComponentMask(None=NamedRerouteUsage()) {g=true}
  )
  FunctionOutput(None=SetMaterialAttributes!(MaterialAttributes=@n1.MaterialAttributes, Opacity=LinearInterpolate(A=@n1.Opacity, B=@n2, Alpha=@n3)) {attributeSetTypes [#1]})
  NamedRerouteDeclaration(None=FunctionInput(Preview=TextureSample() {samplerType=SAMPLERTYPE_Masks, texture=/BRCosmetics/Animation/Props/Gadgets_MrMite/Textures/T_MrMite_Gadget_FX}.RGB) {inputName=In_TechArtMask, sortPriority=1}) {name=TechArtMask, nodeColor=<rgba 1 0.172 0.0441 1>}
  NamedRerouteDeclaration(None=FunctionInput(Preview=TextureSample() {samplerType=SAMPLERTYPE_Masks, texture=&T_MrMite_Gadget_FX}.A) {bUsePreviewValueAsDefault=true, inputName=In_TechArtMask_A, inputType=FunctionInput_Scalar, sortPriority=2}) {name=TechArtMask_A, nodeColor=<rgba 1 0.172 0.0441 1>}
  NamedRerouteDeclaration(None=FunctionInput() {inputType=FunctionInput_MaterialAttributes}) {name=In_MaterialAttributes, nodeColor=<rgba 0 1 0.6953 1>}
 ]}
